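# Idle Niagara effect for the Peely Tech character body, adding electric-looking arcs, sparks and airglow to its skeletal mesh.
# Randomly looping ribbon bursts follow socket-driven bone-flow splines, while brief socket sparks flash with lights and emit death events.
# A separate emitter defines cone-thrown sparks, and persistent glows track two sockets.
# The head-hidden flag gates arc alpha, lights and visibility tags; the distortion emitter is disabled.
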NiagaraSystem /BRCosmetics/Characters/Player/Male/Medium/Bodies/M_MED_Peely_Tech/FX/NS_Peely_tech_idle
dependencies {
 usedDataInterfaces [
  &NiagaraDataInterfaceVector2DCurve
  &NiagaraDataInterfaceCurve
  &NiagaraDataInterfaceArrayInt32
  &NiagaraDataInterfaceColorCurve
 ]
 usedDynamicInputs [&Vector4FromCurve]
 usedModules [&SystemState]
}
emitters {
 Bone_Flow_Ribbon02 {
  emitterUpdateScript [
   &EmitterState(
    Life Cycle Mode=::NewEnumerator1
    Loop Duration=&RandomRangeFloat(Minimum=0.5, Maximum=5)
    Recalculate Duration Each Loop=true
    UseLoopDelay=true
   )
   /SetVariables(
    Emitter.Spawn Count:NiagaraInt32=&RandomRangeInt(Minimum=3, Maximum=6, Evaluation Type=::NewEnumerator1)
    Emitter.genRand:NiagaraFloat=&RandomRangeFloat(Minimum=0.15, Maximum=0.3, Evaluation Type=::NewEnumerator1)
   ) {isSetParameters=true}
   &SpawnBurst_Instantaneous(Spawn Count=$Emitter.Spawn Count, Use Spawn Probability=false, UseLoopCountLimit=false)
   &SpawnBurst_Instantaneous(Spawn Count=6, Spawn Time=0.6234, Use Spawn Probability=false, UseLoopCountLimit=false) {enabled=false, module=SpawnBurst_Instantaneous001}
  ]
  particleSpawnScript [
   &FN_SetSpawnGroup(
    Inclusivity=true
    Normalize Spawn Group=true
    Particles Per Group:NiagaraInt32=$Emitter.Spawn Count
    Pattern=::NewEnumerator0
    Set Ribbon ID and Ribbon Link Order=true
   )
   /SetVariables(Particles.random:NiagaraFloat=$Emitter.genRand) {isSetParameters=true}
   &InitializeRibbon(Color=<rgba 10 5 0 1>, Lifetime=$Particles.random, Ribbon Width=12, UsePositionOffset=false, Write Material Random=true)
   &CameraOffset(Camera Offset Amount=10)
   &DynamicMaterialParameters(
    Index 0 Param 1=1
    Index 0 Param 3=&RandomRangeFloat(Minimum=0.25, Maximum=0.75)
    Index 0 Param 4=1
    Module.Param0WriteEnabled=true
    Module.Param1WriteEnabled=true
    Module.Param2WriteEnabled=true
    Module.Param3WriteEnabled=true
   )
  ]
  particleUpdateScript [
   &ParticleState()
   &FN_BoneFlow(
    Angle:NiagaraFloat=&RandomRangeFloat(Minimum=0, Maximum=0.1)
    BoneID=0
    Enable Position Offsets=true
    Height:NiagaraFloat=$Particles.Normalized Spawn Group
    Offset Mid:Vector3f=&Lerp_Vector(
     A=<0 0 0>
     B=&RandomRangeVector(
      Minimum=<0 0 0>
      Maximum=&RandomVector(Vector Scale=&RandomRangeFloat(Minimum=0, Maximum=0))
     )
     Alpha=$Particles.NormalizedAge
    )
    Radius:NiagaraFloat=&Multiply_Float(
     A=&FloatFromCurve(
      FloatCurve=DataInterface() {propertyValues {Curve {keys [
          {arriveTangent=4.3816, interpMode=RCIM_Cubic, leaveTangent=4.3816, tangentMode=RCTM_User, time=0, value=0}
          {interpMode=RCIM_Cubic, tangentMode=RCTM_User, time=0.5, value=1}
          {arriveTangent=-4.6838, interpMode=RCIM_Cubic, leaveTangent=-4.6838, tangentMode=RCTM_User, time=1, value=0}
         ]}}}
      CurveIndex=$Particles.Normalized Spawn Group
      Scale Curve=&RandomRangeFloat(Minimum=0, Maximum=2)
     )
     B=&Lerp_Float(
      A=&RandomRangeFloat(Minimum=0, Maximum=1)
      B=&RandomRangeFloat(Minimum=0, Maximum=1)
      Alpha=$Particles.NormalizedAge
     )
    )
    Set Material Dynamic=false
    SkeletalMesh:NiagaraDataInterfaceSkeletalMesh=$User.SkeletalMesh
    Use Sockets Instead=true
   )
   /SetVariables(Particles.RibbonFacing=$Output.FN_BoneFlow.SplineNormal) {isSetParameters=true}
   &ScaleRibbonWidth(Ribbon Width Scale=&Multiply_Float(
     A=&Multiply_Float(
      A=&FloatFromCurve(
       FloatCurve=DataInterface() {propertyValues {Curve {keys [
           {interpMode=RCIM_Cubic, tangentMode=RCTM_User, time=0, value=0}
           <at 0.1 0>
           {interpMode=RCIM_Cubic, tangentMode=RCTM_Break, time=0.2, value=1}
           {arriveTangent=-1.5728, interpMode=RCIM_Cubic, leaveTangent=-1.5728, tangentMode=RCTM_User, time=1, value=0.5}
          ]}}}
       Scale Curve=&RandomRangeFloat(Minimum=0.7, Maximum=1)
      )
      B=0.75
     )
     B=&FloatFromCurve(
      FloatCurve=DataInterface() {propertyValues {Curve {keys [
          {interpMode=RCIM_Cubic, tangentMode=RCTM_User, time=0, value=0.675}
          {interpMode=RCIM_Cubic, tangentMode=RCTM_User, time=0.5, value=0.5}
          {tangentMode=RCTM_User, time=1, value=0.675}
         ]}}}
      CurveIndex=$Particles.Normalized Spawn Group
      Scale Curve=&FloatFromCurve(
       FloatCurve=DataInterface() {propertyValues {Curve {keys [
           {interpMode=RCIM_Cubic, tangentMode=RCTM_User, time=0, value=1}
           {interpMode=RCIM_Cubic, tangentMode=RCTM_User, time=0.3, value=0.4}
           {interpMode=RCIM_Cubic, tangentMode=RCTM_User, time=0.5, value=1.5}
           {interpMode=RCIM_Cubic, tangentMode=RCTM_User, time=0.7, value=0.4}
           {interpMode=RCIM_Cubic, tangentMode=RCTM_User, time=1, value=1}
          ]}}}
       CurveIndex=&Multiply_Float(
        A=$Particles.Normalized Spawn Group
        B=&Multiply_Float(A=$Particles.Age, B=&RandomRangeFloat(Minimum=1, Maximum=1.5))
       )
      )
     )
    )) {enabled=false}
   &ScaleRibbonWidth(Ribbon Width Scale=&Multiply_Float(
     A=&FloatFromCurve(FloatCurve=DataInterface() {propertyValues {Curve {keys [<at 0 0>,<at 0.1 1.5>,<at 0.2 1>,<at 1 0>]}}})
     B=&FloatFromCurve(
      FloatCurve=DataInterface() {propertyValues {Curve {keys [
          {arriveTangent=1.3827, interpMode=RCIM_Cubic, leaveTangent=1.3827, tangentMode=RCTM_User, time=0, value=0.5}
          {interpMode=RCIM_Cubic, tangentMode=RCTM_User, time=0.5, value=1}
          {arriveTangent=-2.0974, interpMode=RCIM_Cubic, leaveTangent=-2.0974, tangentMode=RCTM_User, time=1, value=0.5}
         ]}}}
      CurveIndex=&ReturnNormalizedExecIndex()
     )
    )) {module=ScaleRibbonWidth001}
   &ScaleColor(
    Scale Alpha=&FloatFromCurve(
     FloatCurve=DataInterface() {propertyValues {Curve {keys [
         {arriveTangent=8.0553, interpMode=RCIM_Cubic, leaveTangent=8.0553, tangentMode=RCTM_User, time=0, value=0}
         <at 0.2113 1 RCIM_Cubic>
         <at 0.4546 0>
         {arriveTangent=-8.0018, interpMode=RCIM_Cubic, leaveTangent=7.1228, tangentMode=RCTM_Break, time=0.5517, value=0}
         <at 0.7524 1 RCIM_Cubic>
         {arriveTangent=-10.4882, interpMode=RCIM_Cubic, leaveTangent=-10.4882, tangentMode=RCTM_User, time=1, value=0}
        ]}}}
     Scale Curve=&MakeCustomFloatFromBool(Bool=$User.bisHeadHidden, True Float=0, False Float=1)
    )
    ScaleA=true
    ScaleRGB=true
   ) {module=ScaleColor001}
   &SolveForcesAndVelocity(Clamp Velocity=false, Limit Acceleration=false)
   &FN_Light_Attributes(
    Color=<rgba 1 0.625 0.25 1>
    Create Color Attribute=true
    Create Exponent Attribute=true
    Create Position Attribute=true
    Create Radius Attribute=true
    Enable Particle Light Enable=true
    Exponent=1
    Light Enabled:NiagaraBool=&Boolean_NotOperation(Bool=$User.bisHeadHidden)
    Radius=10
    Scale Color=&FloatFromCurve(FloatCurve=DataInterface() {propertyValues {Curve {keys [
         {arriveTangent=6.063, interpMode=RCIM_Cubic, leaveTangent=6.063, tangentMode=RCTM_User, time=0, value=0}
         {arriveTangent=0.5, interpMode=RCIM_Cubic, leaveTangent=0.5, time=0.2, value=0.9}
         {interpMode=RCIM_Cubic, tangentMode=RCTM_User, time=0.5, value=0.25}
         {arriveTangent=-0.5, interpMode=RCIM_Cubic, leaveTangent=-0.5, time=0.8, value=0.7}
         {arriveTangent=-5.6718, interpMode=RCIM_Cubic, leaveTangent=-5.6718, tangentMode=RCTM_User, time=1, value=0}
        ]}}})
   )
  ]
  rendererClasses [&NiagaraRibbonRendererProperties, &NiagaraLightRendererProperties]
  renderers [
   {rendererClass=&NiagaraRibbonRendererProperties}
   {rendererClass=&NiagaraLightRendererProperties, rendererIndex=1}
  ]
 }
 Distortion_modify {
  bEnabled=false
  emitterUpdateScript [
   &EmitterState()
   &SpawnBurst_Instantaneous(Spawn Count=3, Use Spawn Probability=false) {enabled=false}
   &SpawnRate(
    SpawnRate=&FloatFromCurve(FloatCurve=DataInterface() {propertyValues {Curve {keys [<at 0 1>,<at 1 0>]}}}, CurveIndex=$Emitter.NormalizedLoopAge, Scale Curve=20)
    Use Spawn Probability=false
   )
  ]
  particleSpawnScript [
   /SetVariables(Particles.Random=&RandomRangeFloat(Minimum=0, Maximum=1)) {isSetParameters=true}
   &FN_SetSpawnGroup_v2(Number of Spawn Groups=3)
   &InitializeParticle(
    Color=<rgba 0 0.5 0 1>
    Lifetime=0.4
    Position Offset=&MakeVector(
     X=0
     Y=0
     Z=&Multiply_Float(A=&MakeFloatFromInt(Int=&SpawnGroup()), B=7)
    )
    Sprite Rotation Angle=90
    Sprite Rotation Mode=::NewEnumerator0
    Sprite Size Max=<8 45>
    Sprite Size Min=<5 35>
    Sprite Size Mode=::NewEnumerator2
    UsePositionOffset=true
    Write Material Random=true
   ) {module=InitializeParticle001}
   &SphereLocation(
    MaskSpawnGroup=false
    Override Seed=false
    Sphere Radius=15
    Surface Only=false
    UseNonUniformScale=false
    UseOffset=false
    UseSphereOrientation=false
   )
   &CameraOffset(Camera Offset Amount=25)
   &SystemLocation(MaskSpawnGroup=false, Offset=<0 0 50>) {enabled=false}
   &DynamicMaterialParameters(
    Index 0 Param 1=0.1
    Index 0 Param 2=0.5
    Index 0 Param 3=-2
    Module.Param0WriteEnabled=true
    Module.Param1WriteEnabled=true
    Module.Param2WriteEnabled=true
    Module.Param3WriteEnabled=true
   )
  ]
  particleUpdateScript [
   &ParticleState()
   &DynamicMaterialParameters(
    Index 1 Param 1=&FloatFromCurve(FloatCurve=DataInterface() {propertyValues {Curve {keys [<at 0 1>,<at 0.4446 1>,<at 1 0>]}}}, Scale Curve=5)
    Module.Param0WriteEnabled1=true
    Module.Param1WriteEnabled1=true
    Module.Param2WriteEnabled1=true
    Module.Param3WriteEnabled1=true
    Write Parameter Index 0=false
    Write Parameter Index 1=true
   ) {module=DynamicMaterialParameters001}
   &SolveForcesAndVelocity(Clamp Velocity=false, Limit Acceleration=false)
  ]
  rendererClasses [&NiagaraSpriteRendererProperties]
  renderers [{rendererClass=&NiagaraSpriteRendererProperties}]
 }
 Initial_Spark {
  emitterUpdateScript [
   &EmitterState(
    Life Cycle Mode=::NewEnumerator1
    Loop Duration=&Lerp_Float(
     A=0.5
     B=3
     Alpha=&FloatFromCurve(
      FloatCurve=DataInterface() {propertyValues {Curve {keys [
          {
           arriveTangent=0.0545
           arriveTangentWeight=0.4647
           interpMode=RCIM_Cubic
           leaveTangent=0.0545
           leaveTangentWeight=0.4647
           tangentMode=RCTM_User
           tangentWeightMode=RCTWM_WeightedBoth
           time=0
           value=0
          }
          {
           arriveTangent=0.3758
           arriveTangentWeight=0.9371
           interpMode=RCIM_Cubic
           leaveTangent=0.3758
           leaveTangentWeight=0.9371
           tangentMode=RCTM_User
           tangentWeightMode=RCTWM_WeightedBoth
           time=1
           value=1
          }
         ]}}}
      CurveIndex=&RandomRangeFloat(Minimum=0, Maximum=1)
     )
    )
    Recalculate Duration Each Loop=true
    UseLoopDelay=false
   )
   &SpawnBurst_Instantaneous(Spawn Count=1, Use Spawn Probability=false) {module=SpawnBurst_Instantaneous001}
  ]
  particleSpawnScript [
   &InitializeParticle(
    Color=<rgba 10 5 0 1>
    Lifetime Max=0.025
    Lifetime Min=0.05
    Lifetime Mode=::NewEnumerator1
    Sprite Rotation Mode=::NewEnumerator2
    Sprite Size Mode=::NewEnumerator1
    Sprite UV Mode=::NewEnumerator3
    Uniform Sprite Size Max=2.5
    Uniform Sprite Size Min=4
    UsePositionOffset=true
    Write Material Random=true
   )
   &CameraOffset(Camera Offset Amount=2)
   /SetVariables(
    Particles.Corona Cam offset=10
    Particles.Corona Sprite Size:Vector2f=&Vector2DFromFloat(Value=10)
    Particles.Corona color:LinearColor=<rgba 10 5 0 0.05>
   ) {isSetParameters=true}
   /SetVariables(Particles.Random Int:NiagaraInt32=&SelectIntFromArray(
     Array Sampling Mode=::NewEnumerator1
     Direct Array Index=?
     Int Selection Array=DataInterface() {propertyValues {IntData [0,2]}}
     Random Seed=0
    )) {isSetParameters=true}
  ]
  particleUpdateScript [
   &ParticleState()
   &SkeletalMeshLocation(
    Bone / Socket Index=$Particles.Initial.Random Int
    MaskSpawnGroup=false
    Mesh Sampling Type=::NewEnumerator5
    Skeletal Mesh=$User.SkeletalMesh
    Socket Sampling Mode=::NewEnumerator5
    Velocity Sampling=::NewEnumerator0
    VelocitySpeedLimit=false
   )
   &ScaleSpriteSize(Scale Factor=&Vector2DFromFloat(Value=&FloatFromCurve(FloatCurve=DataInterface() {propertyValues {Curve {keys [<at 0 0>,<at 0.7 1.5>,<at 1 0.2>]}}})))
   &SolveForcesAndVelocity(Clamp Velocity=false, Limit Acceleration=false)
   &GenerateDeathEvent(Float to Send as Random Float:NiagaraFloat=&MakeFloatFromInt(Int=$Particles.Initial.Random Int))
   &FN_Light_Attributes(
    Color=<rgba 2 1 0 1>
    Create Color Attribute=true
    Create Exponent Attribute=true
    Create Position Attribute=true
    Create Radius Attribute=true
    Enable Particle Light Enable=true
    Exponent=2
    Light Enabled:NiagaraBool=&Boolean_NotOperation(Bool=$User.bisHeadHidden)
    Position=&Add_Vector(
     A=Linked() {linkedVariable {name=Particles.Position, type=NiagaraPosition}}
     B=&TransformVector(Vector=<-2 -5 0>, Destination Space=::Simulation)
    )
    Radius=15
    Scale Color=&FloatFromCurve(
     FloatCurve=DataInterface() {propertyValues {Curve {keys [
         {arriveTangent=5.2047, interpMode=RCIM_Cubic, leaveTangent=5.2047, tangentMode=RCTM_User, time=0, value=0}
         {interpMode=RCIM_Cubic, tangentMode=RCTM_User, time=0.5, value=1}
         {arriveTangent=-4.404, interpMode=RCIM_Cubic, leaveTangent=-4.404, tangentMode=RCTM_User, time=1, value=0}
        ]}}}
     Scale Curve=2
    )
   )
   /SetVariables(Particles.VisibilityTag=&MakeCustomIntFromBool(Bool=$User.bisHeadHidden, True INT=1, False INT=0)) {isSetParameters=true, module=SetVariables_#1}
  ]
  rendererClasses [&NiagaraSpriteRendererProperties, &NiagaraLightRendererProperties]
  renderers [
   {rendererClass=&NiagaraSpriteRendererProperties}
   {rendererClass=&NiagaraSpriteRendererProperties, rendererIndex=1}
   {rendererClass=&NiagaraLightRendererProperties, rendererIndex=2}
  ]
 }
 Lights_Airglow {
  emitterUpdateScript [
   &EmitterState()
   &SpawnBurst_Instantaneous(Age:NiagaraFloat=$Emitter.Age, Spawn Count=2, Spawn Time=0.01, Use Spawn Probability=false, UseLoopCountLimit=false)
  ]
  particleSpawnScript [
   /SetVariables(Particles.Size=75) {isSetParameters=true}
   &InitializeParticle(
    Color=<rgba 1 0.1667 0 0.25>
    Lifetime=0.1
    Mesh Scale=<0.1 0.125 0.115>
    Mesh Scale Mode=::NewEnumerator4
    Sprite Size Mode=::NewEnumerator3
    Uniform Sprite Size=&Divide_Float(A=$Particles.Size, B=4)
    UsePositionOffset=false
    Write Material Random=true
   )
   &CameraOffset(Camera Offset Amount=8)
  ]
  particleUpdateScript [
   &ParticleState(
    Kill Particles When Lifetime Has Elapsed=false
    Let Infinitely Lived Particles Die When Emitter Deactivates=true
    Loop Particles Lifetime=true
   )
   &DynamicMaterialParameters(
    Index 0 Param 1=&Lerp_Float(A=1.5, B=3, Alpha=1)
    Index 0 Param 2=1
    Module.Param0WriteEnabled=true
    Module.Param1WriteEnabled=true
    Module.Param2WriteEnabled=true
    Module.Param3WriteEnabled=true
   )
   &SkeletalMeshLocation(
    Bone / Socket Index=&Add_Integer(A=&ReturnExecIndex(), B=3)
    MaskSpawnGroup=false
    Mesh Sampling Type=::NewEnumerator5
    Sampled Position Offset=<-1 0 -0.25>
    Skeletal Mesh=$User.SkeletalMesh
    Socket Sampling Mode=::NewEnumerator5
   )
   &SpriteFacingAndAlignment(
    Sprite Facing=&Multiply_VectorByFloat(
     Vector=&ReturnMeshOrientationAxis(Mesh Orientation Axis=::XAxis, Mesh Orientation Quat=$Particles.SkeletalMeshLocation.SampledRotation)
     Float=1
    )
    WriteAlignment=false
    WriteFacing=true
   ) {enabled=false}
   /SetVariables(Particles.LightIntencity:NiagaraFloat=&Lerp_Float(A=1, B=2, Alpha=1)) {isSetParameters=true}
   &FN_Light_Attributes(
    Color=<rgba 1 0.18 0 1>
    Create Color Attribute=true
    Create Exponent Attribute=true
    Create Position Attribute=true
    Create Radius Attribute=true
    Exponent=2
    Position=&Add_Vector(
     A=Linked() {linkedVariable {name=Particles.Position, type=NiagaraPosition}}
     B=&Add_Vector(
      A=&Multiply_VectorByFloat(
       Vector=&ReturnMeshOrientationAxis(Mesh Orientation Axis=::XAxis, Mesh Orientation Quat=$Particles.SkeletalMeshLocation.SampledRotation)
       Float=5
      )
      B=&Add_Vector(
       A=&Multiply_VectorByFloat(
        Vector=&ReturnMeshOrientationAxis(Mesh Orientation Axis=::ZAxis, Mesh Orientation Quat=$Particles.SkeletalMeshLocation.SampledRotation)
        Float=-2.5
       )
       B=&Multiply_VectorByFloat(
        Vector=&ReturnMeshOrientationAxis(Mesh Orientation Axis=::YAxis, Mesh Orientation Quat=$Particles.SkeletalMeshLocation.SampledRotation)
        Float=&Lerp_Float(A=0, B=0, Alpha=&ReturnNormalizedExecIndex())
       )
      )
     )
    )
    Radius=12
    Scale Color=$Particles.LightIntencity
   ) {enabled=false}
   &ScaleColor(ScaleA=true, ScaleRGB=true)
   /SetVariables(Particles.VisibilityTag=&MakeCustomIntFromBool(Bool=$User.bisHeadHidden, True INT=1, False INT=0)) {isSetParameters=true, module=SetVariables_#1}
  ]
  rendererClasses [
   &NiagaraSpriteRendererProperties
   &NiagaraLightRendererProperties
   &NiagaraMeshRendererProperties
  ]
  renderers [
   {rendererClass=&NiagaraSpriteRendererProperties}
   {rendererClass=&NiagaraLightRendererProperties, rendererIndex=1}
   {rendererClass=&NiagaraMeshRendererProperties, rendererIndex=2}
  ]
 }
 Sparks001 {
  emitterUpdateScript [&EmitterState(
    Life Cycle Mode=::NewEnumerator1
    Loop Delay=&RandomRangeFloat(Minimum=0.1, Maximum=0.2)
    Loop Duration=1.15
    Recalculate Duration Each Loop=true
    UseLoopDelay=false
   )]
  particleSpawnScript [
   &InitializeParticle(
    Color=<rgba 1 0.824 0.2116 1>
    Lifetime=&RandomRangeFloat(Minimum=0.1, Maximum=0.3)
    Mass Mode=::NewEnumerator1
    Sprite Rotation Mode=::NewEnumerator2
    Sprite Size=<1 8>
    Sprite Size Mode=::NewEnumerator4
    UsePositionOffset=false
    Write Material Random=true
   )
   &SkeletalMeshLocation(
    Bone / Socket Index=2
    MaskSpawnGroup=false
    Mesh Sampling Type=::NewEnumerator5
    Orientation Sampling=::NewEnumerator1
    Position Sampling=::NewEnumerator2
    Skeletal Mesh=$User.SkeletalMesh
    Socket Sampling Mode=::NewEnumerator5
   )
   &AddVelocityInCone(
    Cone Angle=60
    Cone Axis=<1 -0.7 0>
    Module.Use Velocity Falloff On Cone Axis=false
    Velocity Strength=&RandomRangeFloat(Minimum=300, Maximum=500)
   )
   &InheritVelocity(Inherited Velocity Speed Limit=1000000000, Speed Limit=true, UseSpeedThreshold=false) {enabled=false}
   /SetVariables() {isSetParameters=true}
  ]
  particleUpdateScript [
   &ParticleState()
   &Drag(Drag=30, Use Linear Drag=true, Use Rotational Drag=false)
   &ScaleSpriteSize(Scale Factor=&Vector2DFromCurve(
     Vector2Curve=DataInterface() {propertyValues {
       XCurve {keys [
         {
          interpMode=RCIM_Cubic
          leaveTangent=-3.2585
          leaveTangentWeight=0.9206
          tangentMode=RCTM_Break
          tangentWeightMode=RCTWM_WeightedBoth
          time=0
          value=1
         }
         {arriveTangentWeight=0.3333, interpMode=RCIM_Cubic, tangentMode=RCTM_User, tangentWeightMode=RCTWM_WeightedBoth, time=1, value=0}
        ]}
       YCurve {keys [{interpMode=RCIM_Cubic, leaveTangent=-1.6549, leaveTangentWeight=0.383, tangentMode=RCTM_Break, time=0, value=1}, <at 1 0>]}
      }}
     ScaleCurve=<1 1>
    )) {module=ScaleSpriteSize001}
   &GravityForce()
   &SolveForcesAndVelocity(Clamp Velocity=false, Limit Acceleration=false)
  ]
  rendererClasses [&NiagaraSpriteRendererProperties]
  renderers [{rendererClass=&NiagaraSpriteRendererProperties}]
 }
}
summary {
 systemName=NS_Peely_tech_idle
 emitters [
  {emitterName=Sparks001, rendererClasses [&NiagaraSpriteRendererProperties]}
  {emitterName=Initial_Spark, rendererClasses [&NiagaraSpriteRendererProperties, &NiagaraLightRendererProperties]}
  {bEnabled=false, emitterName=Distortion_modify, rendererClasses [&NiagaraSpriteRendererProperties]}
  {
   emitterName=Bone_Flow_Ribbon02
   rendererClasses [&NiagaraRibbonRendererProperties, &NiagaraLightRendererProperties]
  }
  {
   emitterName=Lights_Airglow
   rendererClasses [
    &NiagaraSpriteRendererProperties
    &NiagaraLightRendererProperties
    &NiagaraMeshRendererProperties
   ]
  }
 ]
 userVariables [
  User.bIsFrontendPreview {type=NiagaraBool}
  User.bisHeadHidden {type=NiagaraBool}
  User.HasImpacted {type=NiagaraBool}
  User.SkeletalMesh {
   type=NiagaraDataInterfaceSkeletalMesh
   defaultValue {
    struct=NiagaraExt_VariableValue_DataInterface
    value {dataInterface=&self:NiagaraDataInterfaceSkeletalMesh_1}
   }
  }
  User.UsesParticleLights {type=NiagaraFloat, defaultValue=0, description=""}
 ]
}
