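# Character material for Peely Tech's eyes, drawing an animated, glowing robotic LED-screen face within the shared character PreFX/PostFX material functions.
# Optional flipbook switching rounds facial pose parameters into UV offsets that select eye expressions, with eye shift, hashed time-based glitches and bump offset.
# A camera-facing sphere mask around eye positions tints the eyes, and a cosine grid pixelates them.
# Pulsing glow, scanlines, panning static and sweep lines feed Emissive.
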Material /BRCosmetics/Characters/Player/Male/Medium/Bodies/M_MED_Peely_Tech/Materials/M_PeelyTech_Eyes
graph {
 properties {
  bUseMaterialAttributes=true
  bUsedWithMeshDeformer=true
  bUsedWithSkeletalMesh=true
  pixelDepthOffsetMode=PDOM_Legacy
  usedShadingModels=Default Lit
 }
 nodes [
  ChannelMaskParameter() {parameterName=Glitch Displacement Channel}
  n1 = MaterialFunctionCall() {materialFunction=&BaseFN_CharacterFunction_PreFX}
  n2 = GetMaterialAttributes(None=@n1.Result) {attributeGetTypes [#1]}
  n3 = NamedRerouteUsage()
  n4 = Round(None=Multiply(
    A=ChannelMaskParameter(None=@n3) {maskChannel=Green, parameterName=Glitch Direction, defaultValue=<rgba 0 1 0 0>}.R
    B=ScalarParameter() {defaultValue=132, parameterName=RB_GlitchLine_Tiling}
   ))
  n5 = ComponentMask(None=MaterialFunctionCall(Seed=Time(), Float=@n4, Vec2=@n4) {materialFunction=&MF_Hash}.Vec2) {r=true}
  n6 = ConstantBiasScale(None=Multiply(A=@n5, B=@n5)) {bias=-0.5, scale=2}
  n7 = ComponentMask(None=Multiply(A=NamedRerouteUsage()) {constB=2}) {r=true}
  n8 = Add(A=Time(), B=LinearInterpolate(Alpha=Round(None=@n7)) {constB=0.2})
  n9 = ScalarParameter() {defaultValue=0.1, parameterName=RB_GlitchFrequency}
  n10 = ConstantBiasScale(None=MaterialFunctionCall(Float=Time()) {materialFunction=&MF_Hash}.Float)
  n11 = Subtract(
   A=Frac(None=Multiply(A=@n8, B=@n9) {constB=0.05})
   B=LinearInterpolate(
    A=ScalarParameter() {defaultValue=0.8, parameterName=RB_GlitchThreshold_Min}
    B=ScalarParameter() {defaultValue=0.9, parameterName=RB_GlitchThreshold_Max}
    Alpha=Round(None=@n10)
   )
  )
  n12 = Multiply(A=ScalarParameter() {defaultValue=0.01, parameterName=RB_GlitchDistortion_Amt}, B=Saturate(None=Ceil(None=@n11)))
  n13 = Normalize(VectorInput=ComponentMask(None=VectorParameter() {parameterName=Glitch Displacement Direciton, defaultValue=<rgba 1 0 0 0>}.RGB) {g=true, r=true})
  n14 = Add(
   A=NamedRerouteUsage()
   B=MaterialFunctionCall(X=ScalarParameter() {group=Banana_Poses, parameterName=Shift Eyes X}, Y=ScalarParameter() {group=Banana_Poses, parameterName=Shift Eyes Y}) {materialFunction=&MakeFloat2}.Result
  )
  n15 = Add(A=Multiply(A=Multiply(A=@n6, B=@n12) {constB=0.01}, B=@n13), B=@n14)
  n16 = Add(A=@n14, B=Constant2Vector() {r=0.5})
  n17 = ScalarParameter() {group=Banana_Poses, parameterName=R_squint_inner_pose}
  n18 = Divide(B=ScalarParameter() {defaultValue=0.5, group=Banana, parameterName=Threshold_squint_inner_pose}) {constA=0.5}
  n19 = Clamp(None=Round(None=Multiply(A=@n17, B=@n18)))
  n20 = Constant2Vector() {g=0.5}
  n21 = ScalarParameter() {defaultValue=0.3, group=Banana, parameterName=Threshold_R_brow_up_pose}
  n22 = Round(None=Multiply(A=ScalarParameter() {group=Banana_Poses, parameterName=R_brow_up_pose}, B=Divide(B=@n21) {constA=0.5}))
  n23 = Multiply(
   A=ScalarParameter() {group=Banana_Poses, parameterName=C_glabella_down_pose}
   B=Divide(B=ScalarParameter() {defaultValue=0.2, group=Banana, parameterName=Threshold_Glabella_down}) {constA=0.5}
  )
  n24 = LinearInterpolate(
   A=LinearInterpolate(A=LinearInterpolate(A=@n15, B=@n16, Alpha=@n19), B=Add(A=@n14, B=@n20), Alpha=Clamp(None=@n22))
   B=Add(A=@n14, B=Constant2Vector() {g=0.5, r=0.5})
   Alpha=Clamp(None=Round(None=@n23))
  )
  n25 = Add(A=@n14, B=Constant2Vector() {g=0.25, r=0.5})
  n26 = ScalarParameter() {group=Banana_Poses, parameterName=C_glabella_up_pose}
  n27 = Divide(B=ScalarParameter() {defaultValue=0.5, group=Banana, parameterName=Threshold_Glabella_up}) {constA=0.5}
  n28 = Clamp(None=Round(None=Multiply(A=@n26, B=@n27)))
  n29 = Constant2Vector() {g=0.25}
  n30 = ScalarParameter() {defaultValue=0.3, group=Banana, parameterName=Threshold_R_blink_pose}
  n31 = Round(None=Multiply(A=ScalarParameter() {group=Banana_Poses, parameterName=R_blink_pose}, B=Divide(B=@n30) {constA=0.5}))
  n32 = StaticSwitchParameter(
   True=LinearInterpolate(A=LinearInterpolate(A=@n24, B=@n25, Alpha=@n28), B=Add(A=@n14, B=@n29), Alpha=Clamp(None=@n31))
   False=NamedRerouteUsage()
  ) {parameterName=Enable Robro Eyes Flipbook}
  n33 = TextureObjectParameter() {
   parameterName=EyesTexture
   samplerType=SAMPLERTYPE_LinearColor
   texture=/BRCosmetics/Characters/Player/Male/Medium/Bodies/M_MED_Peely_Tech/Textures/T_MED_PeelyTech_Eyes_Flipbook_Masks
  }
  n34 = TextureSample(UVs=Add(A=@n32, B=NamedRerouteUsage()), Tex=@n33)
  n35 = ChannelMaskParameter(None=@n34.RGB) {parameterName=EyesChannel}
  n36 = NamedRerouteUsage()
  n37 = Sign(None=Subtract(A=ComponentMask(None=@n36) {r=true}) {constB=0.25})
  n38 = StaticSwitchParameter(
   True=VectorParameter() {parameterName=Eye, defaultValue=<rgba 0 7 157 0>}.RGB
   False=LinearInterpolate(
    A=VectorParameter() {parameterName=Eye_R, defaultValue=<rgba -5.5 7 157 0>}.RGB
    B=VectorParameter() {parameterName=Eye_L, defaultValue=<rgba 5.5 7 157 0>}.RGB
    Alpha=Saturate(None=@n37)
   )
  ) {parameterName=Use Single Eye Location}
  n39 = SphereMask(
   A=Constant2Vector()
   B=MaterialFunctionCall(
    Plane Normal Axis=MaterialFunctionCall() {materialFunction=&CameraDirectionVector}.Result
    Plane Center=TransformPosition(None=@n38) {transformType=TRANSFORMPOSSOURCE_World}
   ) {materialFunction=&VirtualPlaneCoordinates}.UVs
   Radius=ScalarParameter() {defaultValue=5, parameterName=Radius}
   Hardness=Constant()
  )
  n40 = Multiply(A=@n35.R, B=@n39)
  n41 = LinearInterpolate(
   A=VectorParameter() {parameterName=Eye Inside Color 2, defaultValue=<rgba 1 0.8333 0 1>}.RGB
   B=VectorParameter() {parameterName=Eye Inside Color 1, defaultValue=<rgba 1 1 1 1>}.RGB
   Alpha=Power(Base=@n39)
  )
  n42 = VectorParameter() {parameterName=RB_EyeColor, defaultValue=<rgba 1 0.8333 0 1>}
  n43 = ChannelMaskParameter(None=@n34.RGB) {maskChannel=Blue, parameterName=EyesGlowChannel, defaultValue=<rgba 0 0 1 0>}
  n44 = Multiply(
   A=LinearInterpolate(
    A=@n42.RGB
    B=VectorParameter() {parameterName=RB_EyeColor2, defaultValue=<rgba 1 0.3333 0 1>}.RGB
    Alpha=Power(Base=@n43.R) {constExponent=0.75}
   )
   B=@n43.R
  )
  n45 = Multiply(A=@n42.RGB, B=@n35.R)
  n46 = ScalarParameter() {defaultValue=0.5, parameterName=RB_EyeGlowStrength_Min}
  n47 = ScalarParameter() {defaultValue=1, parameterName=RB_EyeGlowStrength_Max}
  n48 = Time()
  n49 = ConstantBiasScale(None=Sine(None=Multiply(A=@n48) {constB=0.5}))
  n50 = Power(Base=Multiply(A=Add(A=@n44, B=@n45), B=LinearInterpolate(A=@n46, B=@n47, Alpha=@n49)))
  n51 = NamedRerouteUsage()
  n52 = VectorParameter() {parameterName=Pixel Size, defaultValue=<rgba 132 132 0 0>}
  n53 = Cosine(None=Multiply(A=Add(A=@n32, B=@n51) {constA=1, constB=0}, B=ComponentMask(None=@n52.RGB) {g=true, r=true})) {period=2}
  n54 = Abs(None=ComponentMask(None=@n53) {r=true})
  n55 = Abs(None=ComponentMask(None=@n53) {g=true})
  n56 = Multiply(A=Power(Base=Multiply(A=@n54, B=@n55)) {constExponent=0.5}, B=Add(A=Multiply(A=@n40, B=@n41), B=Add(A=@n2.Emissive Color, B=@n50)))
  n57 = VectorParameter() {parameterName=Screen Bounding Box UV Space, defaultValue=<rgba 0.1 0.005 0.61 0.45>}
  n58 = Divide(A=Subtract(A=NamedRerouteUsage(), B=AppendVector(A=@n57.R, B=@n57.R)), B=AppendVector(A=@n57.B, B=@n57.B))
  n59 = Saturate(None=Sine(None=Saturate(None=@n58)))
  n60 = Multiply(A=ComponentMask(None=@n59) {g=true}, B=ComponentMask(None=@n59) {r=true})
  n61 = NamedRerouteUsage()
  n62 = ConstantBiasScale(None=Frac(None=Panner(Coordinate=@n61) {speedY=-1})) {bias=-0.5, scale=2}
  n63 = Multiply(A=@n62, B=@n62)
  n64 = Add(A=Multiply(
    A=ChannelMaskParameter(None=@n63) {maskChannel=Green, parameterName=Glitch Direction, defaultValue=<rgba 0 1 0 0>}.R
    B=ScalarParameter() {defaultValue=0.5, group=Background, parameterName=Scanline Strength}
   ) {constB=0.1})
  n65 = Constant() {r=6.2832}
  n66 = Constant()
  n67 = Rotator(
   Coordinate=NamedRerouteUsage()
   Time=ChannelMaskParameter(None=AppendVector(A=@n65, B=@n66)) {maskChannel=Green, parameterName=Glitch Direction, defaultValue=<rgba 0 1 0 0>}.R
  )
  n68 = Multiply(A=TextureSample(UVs=Panner(Coordinate=@n67) {speedX=5, speedY=0.1}) {samplerType=SAMPLERTYPE_Masks, texture=/Game/Missions/Primary/RepairTheShelter/Textures/Cinematic_Screen_Static}.B) {constB=0.1}
  n69 = Multiply(
   A=TextureSample(UVs=Panner(Coordinate=@n67) {speedX=6, speedY=-0.05}) {samplerType=SAMPLERTYPE_Masks, texture=&Cinematic_Screen_Static}.G
   B=TextureSample(UVs=Panner(Coordinate=@n67) {speedX=-3, speedY=-0.03}) {samplerType=SAMPLERTYPE_Masks, texture=&Cinematic_Screen_Static}.R
  )
  n70 = Multiply(A=ChannelMaskParameter(None=NamedRerouteUsage()) {maskChannel=Green, parameterName=Glitch Direction, defaultValue=<rgba 0 1 0 0>}.R) {constB=0.75}
  n71 = Multiply(A=Time()) {constB=-0.2}
  n72 = Power(Base=Frac(None=Add(A=@n70, B=@n71))) {constExponent=20}
  n73 = Add(
   A=Multiply(A=Add(A=@n68, B=@n69), B=ScalarParameter() {defaultValue=6, group=Static, parameterName=Static Scale})
   B=Multiply(A=@n72, B=ScalarParameter() {defaultValue=0.2, group=Sweep Line, parameterName=Sweep Add}) {constB=0.2}
  )
  n74 = ScalarParameter() {defaultValue=0.2, group=Background, parameterName=Background}
  n75 = Multiply(
   A=VectorParameter() {group=Colors, parameterName=Screen Tone, defaultValue=<rgba 0.125 0.0625 0 1>}.RGB
   B=Multiply(A=Multiply(A=@n60, B=@n64), B=Add(A=@n73, B=@n74)) {constA=1}
  )
  n76 = MaterialFunctionCall(
   AO=@n1.AO
   Distance Blend=@n1.AO
   Fresnel=@n1.AO
   Gmap=@n1.AO
   Pre Skinned Local Position=@n1.AO
   Material=SetMaterialAttributes(MaterialAttributes=@n2.MaterialAttributes, Emissive Color=Add(A=@n56, B=@n75)) {attributeSetTypes [#1]}
  ) {materialFunction=&BaseFN_CharacterFunction_PostFX}
  MaterialFunctionCall(In=@n76.Result) {materialFunction=&CINE_MF_CharPostFX}
  n77 = Multiply(A=@n69, B=ScalarParameter() {defaultValue=0.07, group=Static, parameterName=Static UV Distort})
  n78 = Constant()
  n79 = Multiply(A=@n72, B=ScalarParameter() {defaultValue=0.005, group=Sweep Line, parameterName=Sweep UV Distort}) {constB=0.02}
  n80 = NamedRerouteUsage()
  NamedRerouteDeclaration(None=Add(A=AppendVector(A=@n77, B=@n78), B=Multiply(A=@n79, B=@n80))) {name=Distorted UVs, nodeColor=<rgba 0.9453 1 0 1>}
  NamedRerouteDeclaration(None=@n15) {name=Glitch Coords, nodeColor=<rgba 0.2656 1 0 1>}
  NamedRerouteDeclaration(None=@n13) {name=Glitch Direction, nodeColor=<rgba 1 0 0.2812 1>}
  n81 = Multiply(
   A=NamedRerouteUsage()
   B=ComponentMask(None=VectorParameter() {parameterName=TexCoordScale, defaultValue=<rgba 5 9 0 0>}.RGB) {g=true, r=true}
  )
  n82 = ComponentMask(None=VectorParameter() {parameterName=TexCoordOffset, defaultValue=<rgba -0.5 0 0 0>}.RGB) {g=true, r=true}
  NamedRerouteDeclaration(None=BumpOffset(Coordinate=Add(A=@n81, B=@n82), Height=ScalarParameter() {defaultValue=-5, parameterName=BumpOffset})) {name=TexCoord, nodeColor=<rgba 1 0 0.4453 1>}
  NamedRerouteDeclaration(None=LinearInterpolate(A=TextureCoordinate(), B=TextureCoordinate() {coordinateIndex=1}, Alpha=ScalarParameter() {defaultValue=1, parameterName=Tex Coord Index})) {name=Tex Coord Indexed, nodeColor=<rgba 1 0 0.0167 1>}
 ]
}
