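# Physics asset for the M_MED_Peely_Tech male player body cosmetic, setting up secondary motion for a dangling side strap and a zipper.
# Four box bodies form the strap chain and a capsule forms the zipper, each linked with tight twist and moderate swing limits to its master joint.
# Both masters attach to spine_03 with free rotation.
# Capsules on spine_03 and the left thigh supply body shapes, with the thigh left unconstrained.
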PhysicsAsset /BRCosmetics/Characters/Player/Male/Medium/Bodies/M_MED_Peely_Tech/Meshes/M_MED_Peely_Tech_Physics
bodies {
 dyn_side_strap_1_jnt [{shapeType=Box, extentX=6, extentY=1, extentZ=4, shapeName=dyn_side_strap_1_jnt_box, center=<3.223 -0.1983 0.0001>}]
 dyn_side_strap_2_jnt [{shapeType=Box, extentX=6, extentY=1, extentZ=4.5, shapeName=dyn_side_strap_2_jnt_box, center=<3.1269 -0.1985 0.0001>}]
 dyn_side_strap_3_jnt [{shapeType=Box, extentX=5.5, extentY=1.5, extentZ=4.5, shapeName=dyn_side_strap_3_jnt_box, center=<2.8244 -0.1534 0>}]
 dyn_side_strap_4_jnt [{shapeType=Box, extentX=7.5, extentY=1.5, extentZ=4.5, shapeName=dyn_side_strap_4_jnt_box, center=<3.8254 -0.1984 -0.0001>}]
 dyn_side_strap_master_jnt [{shapeType=Capsule, length=1, radius=1, shapeName=dyn_side_strap_master_jnt_capsule}]
 dyn_zipper_jnt [{shapeType=Capsule, length=3, radius=1, shapeName=dyn_zipper_jnt_capsule, center=<2.0613 -0.0002 0.0923>, rotation=<rot 90 -0.0197 -0.0197>}]
 dyn_zipper_master_jnt [{shapeType=Capsule, length=1, radius=1, shapeName=dyn_zipper_master_jnt_capsule}]
 spine_03 [
  {shapeType=Capsule, length=6.75, radius=14.5, shapeName=spine_03_capsule, center=<-8.6433 1.7546 0.3565>}
  {shapeType=Capsule, length=9, radius=15, shapeName=spine_03_capsule, center=<0.8046 0.4661 0.2221>}
 ]
 thigh_l [
  {
   shapeType=Capsule
   length=30.3674
   radius=8.5643
   shapeName=thigh_l_capsule
   center=<-20.6928 1.5881 0.5628>
   rotation=<rot 0.8911 -79.1332 -92.4493>
  }
  {shapeType=Capsule, length=23.25, radius=6.75, shapeName=thigh_l_capsule, center=<0 0.0001 -1.2947>, rotation=<rot -90 0 0>}
 ]
}
constraints [
 {bone1Name=dyn_side_strap_4_jnt, bone2Name=dyn_side_strap_3_jnt, swing1LimitDegrees=45, swing2LimitDegrees=45, twistLimitDegrees=5}
 {bone1Name=dyn_side_strap_2_jnt, bone2Name=dyn_side_strap_1_jnt, swing1LimitDegrees=45, swing2LimitDegrees=45, twistLimitDegrees=5}
 {bone1Name=dyn_side_strap_3_jnt, bone2Name=dyn_side_strap_2_jnt, swing1LimitDegrees=45, swing2LimitDegrees=45, twistLimitDegrees=5}
 {bone1Name=dyn_side_strap_1_jnt, bone2Name=dyn_side_strap_master_jnt, swing1LimitDegrees=25, swing2LimitDegrees=25, twistLimitDegrees=5}
 {
  bone1Name=dyn_side_strap_master_jnt
  bone2Name=spine_03
  swing1LimitDegrees=45
  swing1Motion=Free
  swing2LimitDegrees=45
  swing2Motion=Free
  twistLimitDegrees=45
  twistMotion=Free
 }
 {bone1Name=dyn_zipper_jnt, bone2Name=dyn_zipper_master_jnt, swing1LimitDegrees=25, swing2LimitDegrees=25, twistLimitDegrees=2}
 {
  bone1Name=dyn_zipper_master_jnt
  bone2Name=spine_03
  swing1LimitDegrees=45
  swing1Motion=Free
  swing2LimitDegrees=45
  swing2Motion=Free
  twistLimitDegrees=45
  twistMotion=Free
 }
]
