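# Physics asset for the shoes of the male medium Peely Tech player body, setting up secondary motion for shoe details.
# Small capsule bodies cover the pull, strap and tongue dynamic joints on each shoe, plus both foot bones.
# Each detail joint is constrained directly to its foot with narrow angular limits, giving straps less swing than pulls and tongues and keeping twist very tight on all.
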PhysicsAsset /BRCosmetics/Characters/Player/Male/Medium/Bodies/M_MED_Peely_Tech/Meshes/M_MED_Peely_Tech_Physics_Shoes
bodies {
 dyn_L_shoepull_jnt [{shapeType=Capsule, length=2.25, radius=1, shapeName=dyn_L_shoepull_jnt_capsule, center=<1.8651 0.0001 0>, rotation=<rot -90 -0.0395 0.0394>}]
 dyn_L_shoestrap_jnt [{shapeType=Capsule, length=1, radius=2, shapeName=dyn_L_shoestrap_jnt_capsule, rotation=<rot -79.9987 -0.0003 39.9999>}]
 dyn_L_shoetongue_jnt [{shapeType=Capsule, length=4.5, radius=2.75, shapeName=dyn_L_shoetongue_jnt_capsule, rotation=<rot 90 -0.0395 -0.0393>}]
 dyn_R_shoepull_jnt [{
   shapeType=Capsule
   length=2.5
   radius=1.25
   shapeName=dyn_R_shoepull_jnt_capsule
   center=<-2.1189 0.0004 0>
   rotation=<rot -90 -0.0148 0.0148>
  }]
 dyn_R_shoestrap_jnt [{shapeType=Capsule, length=1, radius=2, shapeName=dyn_R_shoestrap_jnt_capsule, rotation=<rot -69.9997 -89.9992 89.9989>}]
 dyn_R_shoetongue_jnt [{shapeType=Capsule, length=4.5, radius=2.75, shapeName=dyn_R_shoetongue_jnt_capsule, rotation=<rot 90 -0.0198 -0.0199>}]
 foot_l [{shapeType=Capsule, length=1, radius=1, shapeName=foot_l_capsule}]
 foot_r [{shapeType=Capsule, length=1, radius=1, shapeName=foot_r_capsule}]
}
constraints [
 {bone1Name=dyn_L_shoepull_jnt, bone2Name=foot_l, swing1LimitDegrees=15, swing2LimitDegrees=15, twistLimitDegrees=2}
 {bone1Name=dyn_L_shoetongue_jnt, bone2Name=foot_l, swing1LimitDegrees=15, swing2LimitDegrees=15, twistLimitDegrees=2}
 {bone1Name=dyn_L_shoestrap_jnt, bone2Name=foot_l, swing1LimitDegrees=5, swing2LimitDegrees=5, twistLimitDegrees=2}
 {bone1Name=dyn_R_shoepull_jnt, bone2Name=foot_r, swing1LimitDegrees=15, swing2LimitDegrees=15, twistLimitDegrees=2}
 {bone1Name=dyn_R_shoetongue_jnt, bone2Name=foot_r, swing1LimitDegrees=15, swing2LimitDegrees=15, twistLimitDegrees=2}
 {bone1Name=dyn_R_shoestrap_jnt, bone2Name=foot_r, swing1LimitDegrees=5, swing2LimitDegrees=5, twistLimitDegrees=2}
]
