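# Physics asset for the face accessory part of the Peely Tech character body, defining physics bodies for its dangling flaps and wires.
# Each bone gets a single capsule, with spine_05 as the shared parent body.
# The right flap and four wire joints constrain directly to spine_05, while the left flap forms a two-link chain from it.
# Angular limits keep twist tight and allow moderate swing, with the right flap swinging mostly on one axis.
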PhysicsAsset /BRCosmetics/Characters/Player/Male/Medium/Bodies/M_MED_Peely_Tech/Meshes/Parts/M_MED_Peely_Tech_FaceAcc_Physics
bodies {
 dyn_flap_01_l [{
   shapeType=Capsule
   length=4.75
   radius=2.25
   shapeName=dyn_flap_01_l_capsule
   center=<1.6029 2.7342 -0.4818>
   rotation=<rot 4.9814 -29.6212 -81.3159>
  }]
 dyn_flap_02_l [{shapeType=Capsule, length=5.5, radius=3.75, shapeName=dyn_flap_02_l_capsule, center=<5.391 0 0>, rotation=<rot 90 5.0512 5.0513>}]
 dyn_flap_r1 [{shapeType=Capsule, length=7.25, radius=4.75, shapeName=dyn_flap_r1_capsule, center=<8.3017 0.0007 1.4638>, rotation=<rot -80 -0.0001 0>}]
 dyn_wire1_jnt [{
   shapeType=Capsule
   length=3.8126
   radius=1.1725
   shapeName=dyn_wire1_jnt_capsule
   center=<0.6513 -0.1928 -0.4981>
   rotation=<rot 0 68.1499 -78.307>
  }]
 dyn_wire2_jnt [{
   shapeType=Capsule
   length=4.7778
   radius=1.225
   shapeName=dyn_wire2_jnt_capsule
   center=<1.2691 -1.9478 -1.2452>
   rotation=<rot 0 -155.9454 -72.2003>
  }]
 dyn_wire3_jnt [{
   shapeType=Capsule
   length=3.873
   radius=1.0833
   shapeName=dyn_wire3_jnt_capsule
   center=<-0.5686 -0.6242 0.015>
   rotation=<rot 19.9669 -90 -30.9421>
  }]
 dyn_wire4_jnt [{
   shapeType=Capsule
   length=3.6456
   radius=1.5664
   shapeName=dyn_wire4_jnt_capsule
   center=<-0.938 0.5635 0.1826>
   rotation=<rot 0 -165.2415 -53.2843>
  }]
 spine_05 [{shapeType=Capsule, length=2, radius=2}]
}
constraints [
 {bone1Name=dyn_flap_r1, bone2Name=spine_05, swing1LimitDegrees=25, swing2LimitDegrees=2, twistLimitDegrees=2}
 {bone1Name=dyn_flap_01_l, bone2Name=spine_05, swing1LimitDegrees=40, swing2LimitDegrees=40, twistLimitDegrees=2}
 {bone1Name=dyn_flap_02_l, bone2Name=dyn_flap_01_l, swing1LimitDegrees=40, swing2LimitDegrees=40, twistLimitDegrees=2}
 {bone1Name=dyn_wire1_jnt, bone2Name=spine_05, swing1LimitDegrees=25, swing2LimitDegrees=25, twistLimitDegrees=5}
 {bone1Name=dyn_wire2_jnt, bone2Name=spine_05, swing1LimitDegrees=25, swing2LimitDegrees=25, twistLimitDegrees=5}
 {bone1Name=dyn_wire3_jnt, bone2Name=spine_05, swing1LimitDegrees=25, swing2LimitDegrees=25, twistLimitDegrees=5}
 {bone1Name=dyn_wire4_jnt, bone2Name=spine_05, swing1LimitDegrees=25, swing2LimitDegrees=25, twistLimitDegrees=5}
]
